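# Niagara system for the Jellyfish vehicle boost cosmetic: a burst-spawned mesh and sprite body trailed by tentacles, a wide faint ribbon and small glowing sprites.
# All emitters spawn behind System.FinalPosition, with trails pushed backward.
# TentaclesLarge drives five ribbon renderers from one particle chain by writing per-tentacle position, color and width attributes.
# Painted-variant user parameters select colors and drive material parameters on the body and tentacles.
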NiagaraSystem /VehicleCosmetics/Boosts/Boost_Jellyfish/FX/NS_Boost_Jellyfish
dependencies {
 usedDataInterfaces [
  &NiagaraDataInterfaceVectorCurve
  &NiagaraDataInterfaceCurve
  &NiagaraDataInterfaceArrayColor
  &NiagaraDataInterfaceSkeletalMesh
 ]
 usedDynamicInputs []
 usedModules [
  &self:SystemSpawnScript.SystemScriptSource.SystemScriptGraph.NiagaraNodeAssignment_0.SetVariables_#1
  &SystemState
 ]
}
emitters {
 BackRibbon {
  emitterUpdateScript [
   &EmitterState()
   &VC_FilteredSocketCount()
   &VC_RateQualityHandler(Epic Spawn Rate=30, High Spawn Rate=30, Low Spawn Rate=30, Use Custom Rate?=true)
   &VC_SpawnRibbon_v1(SpawnRate=$Emitter.Spawn Rate)
  ]
  particleSpawnScript [
   &VC_PaintedVariantsOverride() {enabled=false}
   &VC_InitializeRibbon_v1(
    Color=&MakeCustomLinearColorFromBool(Bool=$User.bIsPainted, True Color=$User.PaintedColor, False Color=<rgba 0.1 0.5 1 0.2>)
    Custom Offset=true
    Local Vector Offset=&AddVectorToPosition(Position=$System.FinalPosition, Vector=<-80 0 0>)
    Offset Amount=1
    Ribbon Width=130
   )
   &AddVelocity(Velocity=<-2000 0 0>) {module=AddVelocity001}
  ]
  particleUpdateScript [
   &ParticleState()
   &SolveForcesAndVelocity(Clamp Velocity=false, Limit Acceleration=false)
   &VC_UpdateRibbon_v1()
   &ScaleColor(Scale Alpha=0.2, ScaleA=true, ScaleRGB=true)
  ]
  rendererClasses [&NiagaraRibbonRendererProperties]
  renderers [{rendererClass=&NiagaraRibbonRendererProperties}]
 }
 Jellyfish_Entry {
  emitterUpdateScript [
   &EmitterState(
    Inactive Response=::NewEnumerator1
    Life Cycle Mode=::NewEnumerator1
    Loop Behavior=::NewEnumerator1
    Loop Duration Mode=::NewEnumerator1
    UseLoopDelay=false
   )
   &SpawnBurst_Instantaneous(Use Spawn Probability=false, UseLoopCountLimit=false)
   &SpawnRate(SpawnRate=50, Use Spawn Probability=false) {enabled=false}
  ]
  particleSpawnScript [
   &VC_PaintedVariantsOverride(
    BurntSienna=<rgba 0.18 0.0381 0 1>
    Cobalt=<rgba 0 0.0567 0.85 1>
    Crimson=<rgba 0.42 0 0.0114 1>
    DefaultColor=<rgba 0.4762 1 0.138 1>
    ForestGreen=<rgba 0 0.38 0.0192 1>
    Grey=<rgba 0.325 0.325 0.325 1>
    Lime=<rgba 0.27 0.54 0 1>
    Orange=<rgba 0.6724 0.2564 0.0336 1>
    Pink=<rgba 0.5 0.2604 0.4129 1>
    Purple=<rgba 0.2166 0 0.63 1>
    Saffron=<rgba 0.63 0.483 0 1>
    TitaniumWhite=<rgba 0.55 0.55 0.55 1>
   )
   &InitializeParticle(
    Color Mode=::NewEnumerator0
    Mesh Scale Mode=::NewEnumerator3
    Mesh Uniform Scale=6.4
    Position Offset=&AddVectorToPosition(Position=$System.FinalPosition, Vector=<-60 0 0>)
    Position Offset Coordinate Space=::Local
    Sprite Rotation Mode=::NewEnumerator2
    Sprite Size Mode=::NewEnumerator3
    UsePositionOffset=true
    Write Material Random=true
   )
   &AddVelocity(Velocity=<-1500 0 0>) {enabled=false}
   &Color(Color=$Particles.Color)
  ]
  particleUpdateScript [
   &ParticleState(Kill Particles When Lifetime Has Elapsed=false, Loop Particles Lifetime=true)
   &ScaleMeshSize(Scale Factor=&VectorFromCurve(VectorCurve=DataInterface() {propertyValues {XCurve {keys [<at 0 0>,<at 0.6 0.8>,<at 1 1>]}, YCurve {keys [<at 0 1>,<at 1 0>]}, ZCurve {keys [<at 0 1>,<at 1 0>]}}})) {enabled=false}
   &SolveForcesAndVelocity(Clamp Velocity=false, Limit Acceleration=false)
   &DynamicMaterialParameters(
    Index 0 Param 1=&MakeCustomFloatFromBool(Bool=$User.bIsPainted, True Float=0, False Float=1)
    Index 0 Param 3=1
    Index 0 Param 4=HlslExpression() {hlslExpression="User.ColorVariantID == 7 || User.ColorVariantID == 14 ? true : false"}
    Module.Index0Float0Write=true
    Module.Index0Float1Write=true
    Module.Index0Float2Write=true
    Module.Index0Float3Write=true
   )
  ]
  rendererClasses [&NiagaraSpriteRendererProperties, &NiagaraMeshRendererProperties]
  renderers [
   {rendererClass=&NiagaraSpriteRendererProperties}
   {rendererClass=&NiagaraMeshRendererProperties, rendererIndex=1}
  ]
 }
 NE_VC_SingleBoost_Sprite_Parent {
  emitterUpdateScript [
   &EmitterState()
   &SpawnRate(SpawnRate=120, Use Spawn Probability=false)
  ]
  particleSpawnScript [
   &VC_PaintedVariantsOverride() {enabled=false}
   &InitializeParticle(
    Color=&MakeCustomLinearColorFromBool(
     Bool=$User.bIsPainted
     True Color=$User.PaintedColor
     False Color=&SelectLinearColorFromArray(
      Array Sampling Mode=::NewEnumerator0
      Direct Array Index=?
      Color Selection Array=DataInterface() {propertyValues {ColorData [<rgba 0.6 0.1 1 1>, <rgba 0 0.15 1 1>]}}
      Random Seed=0
     )
    )
    Lifetime=&RandomRangeFloat(Minimum=0.25, Maximum=0.5)
    Position Offset=&AddVectorToPosition(Position=$System.FinalPosition, Vector=<-100 0 0>)
    Position Offset Coordinate Space=::Local
    Sprite Rotation Mode=::NewEnumerator2
    Sprite Size Mode=::NewEnumerator3
    Uniform Sprite Size=&RandomRangeFloat(Minimum=3, Maximum=5)
    UsePositionOffset=true
    Write Material Random=true
   )
   &AddVelocity(Velocity=<-1300 0 0>)
   &ShapeLocation(MaskSpawnGroup=false, Override Seed=false, Sphere Radius=45, UseOwnerScale=true)
   &Color(Color=&SelectLinearColorFromArray(
     Array Sampling Mode=::NewEnumerator0
     Direct Array Index=?
     Color Selection Array=DataInterface() {propertyValues {ColorData [<rgba 0.6 0.1 1 1>, <rgba 0 0.15 1 1>]}}
     Random Seed=0
    )) {enabled=false}
  ]
  particleUpdateScript [
   &ParticleState()
   &ScaleSpriteSize(Uniform Curve Sprite Scale=DataInterface() {propertyValues {Curve {keys [{tangentMode=RCTM_User, time=0, value=0}, <at 0.15 1>, <at 0.8 1>, {tangentMode=RCTM_User, time=1, value=0}]}}})
   &SolveForcesAndVelocity(Clamp Velocity=false, Limit Acceleration=false)
   &ScaleColor(
    Scale Alpha=&FloatFromCurve(FloatCurve=DataInterface() {propertyValues {Curve {keys [<at 0.8 1>,<at 1 0>]}}})
    Scale RGB=<100 100 100>
    ScaleA=true
    ScaleRGB=true
   )
  ]
  rendererClasses [&NiagaraSpriteRendererProperties]
  renderers [{rendererClass=&NiagaraSpriteRendererProperties}]
 }
 TentaclesLarge {
  emitterUpdateScript [
   &EmitterState()
   &VC_FilteredSocketCount()
   &VC_RateQualityHandler(Epic Spawn Rate=30, High Spawn Rate=30, Low Spawn Rate=30, Use Custom Rate?=true)
   &VC_SpawnRibbon_v1(SpawnRate=$Emitter.Spawn Rate)
  ]
  particleSpawnScript [
   &VC_PaintedVariantsOverride(
    BurntSienna=<rgba 0.085 0.02 0.0025 1>
    Cobalt=<rgba 0 0 0.85 1>
    Crimson=<rgba 0.4 0 0.0108 1>
    ForestGreen=<rgba 0 0.34 0.0172 1>
    Grey=<rgba 0.28 0.28 0.28 1>
    Lime=<rgba 0.2862 0.54 0.0324 1>
    Orange=<rgba 0.55 0.0642 0 1>
    Pink=<rgba 0.5 0.175 0.3818 1>
    Purple=<rgba 0.05 0 0.4 1>
    Saffron=<rgba 0.6 0.4 0 1>
    TitaniumWhite=<rgba 0.55 0.55 0.55 1>
   )
   &VC_InitializeRibbon_v1(
    Color=$Particles.Color
    Custom Offset=true
    Lifetime=0.7
    Local Vector Offset=&AddVectorToPosition(Position=$System.FinalPosition, Vector=<-80 0 0>)
    Offset Amount=1
    Ribbon Width=12
   )
   &AddVelocity(Velocity=<-2000 0 0>) {module=AddVelocity001}
  ]
  particleUpdateScript [
   &ParticleState()
   &SolveForcesAndVelocity(Clamp Velocity=false, Limit Acceleration=false)
   &VC_UpdateRibbon_v1()
   &DynamicMaterialParameters(
    Index 0 Param 1=&FloatFromCurve(FloatCurve=DataInterface() {propertyValues {Curve {keys [<at 0 6>,<at 1.15 0>]}}})
    Index 0 Param 2=3
    Index 0 Param 3=&FloatFromCurve(FloatCurve=DataInterface() {propertyValues {Curve {keys [<at 0 5 RCIM_Cubic>, {arriveTangent=-13.7143, interpMode=RCIM_Cubic, leaveTangent=-13.7143, time=0.2, value=2}, <at 0.35 0.2 RCIM_Cubic>]}}})
    Index 0 Param 4=&MakeCustomFloatFromBool(Bool=$User.bIsPainted, True Float=0, False Float=1)
    Module.Index0Float0Write=true
    Module.Index0Float1Write=true
    Module.Index0Float2Write=true
    Module.Index0Float3Write=true
   )
   /SetVariables(
    Particles.Tentacle1 Color:LinearColor=&VC_DynamicScaleColor(Color to Scale:LinearColor=$Particles.Initial.Color, RGB Scale:Vector3f=<4 4 4>, Alpha Scale:NiagaraFloat=1)
    Particles.Tentacle1_Position:NiagaraPosition=&AddVectorToPosition(Position=$Particles.Position, Vector=/VehicleCosmetics/FX/Niagara/Modules/VC_ConvertVectorToWorld(Vector:Vector3f=<10 0 15>))
    Particles.Tentacle1_Width:NiagaraFloat=&Multiply_Float(
     A=12
     B=&FloatFromCurve(FloatCurve=DataInterface() {propertyValues {Curve {keys [<at 0 1>,<at 0.75 0>]}}})
    )
   ) {isSetParameters=true}
   /SetVariables(
    Particles.Tentacle4 Color:LinearColor=&VC_DynamicScaleColor(
     Color to Scale:LinearColor=$Particles.Initial.Color
     RGB Scale:Vector3f=<0.5 0.5 0.5>
     Alpha Scale:NiagaraFloat=&FloatFromCurve(FloatCurve=DataInterface() {propertyValues {Curve {keys [<at 0 0>,<at 1 4>]}}})
    )
    Particles.Tentacle4Position:NiagaraPosition=&AddVectorToPosition(Position=$Particles.Position, Vector=&VC_ConvertVectorToWorld(Vector:Vector3f=<-20 -10 -9.5>))
   ) {isSetParameters=true}
   &ScaleRibbonWidth(Ribbon Width Scale=&FloatFromCurve(FloatCurve=DataInterface() {propertyValues {Curve {keys [<at 0 1>,<at 1 0>]}}}))
   /SetVariables(
    Particles.Tentacle2 Color:LinearColor=&VC_DynamicScaleColor(
     Color to Scale:LinearColor=$Particles.Initial.Color
     RGB Scale:Vector3f=<1 1 1>
     Alpha Scale:NiagaraFloat=&FloatFromCurve(FloatCurve=DataInterface() {propertyValues {Curve {keys [<at 0 0>,<at 0.2 0.55>]}}})
    )
    Particles.Tentacle2Position:NiagaraPosition=&AddVectorToPosition(Position=$Particles.Position, Vector=&VC_ConvertVectorToWorld(Vector:Vector3f=<-20 -10 3>))
    Particles.Tentacle2_Width:NiagaraFloat=&Multiply_Float(
     A=12
     B=&FloatFromCurve(FloatCurve=DataInterface() {propertyValues {Curve {keys [<at 0 1>,<at 0.4 0>]}}})
    )
   ) {isSetParameters=true}
   /SetVariables(
    Particles.Tentacle3 Color:LinearColor=&VC_DynamicScaleColor(
     Color to Scale:LinearColor=$Particles.Initial.Color
     RGB Scale:Vector3f=<2.5 2.5 2.5>
     Alpha Scale:NiagaraFloat=&FloatFromCurve(FloatCurve=DataInterface() {propertyValues {Curve {keys [<at 0 0>,<at 1 3>]}}})
    )
    Particles.Tentacle3_Position:NiagaraPosition=&AddVectorToPosition(Position=$Particles.Position, Vector=&VC_ConvertVectorToWorld(Vector:Vector3f=<-25 5 3>))
    Particles.Tentacle3_Width:NiagaraFloat=&Multiply_Float(
     A=12
     B=&FloatFromCurve(FloatCurve=DataInterface() {propertyValues {Curve {keys [<at 0 1>,<at 0.45 0>]}}})
    )
   ) {isSetParameters=true}
   /SetVariables(
    Particles.Tentacle5 Color:LinearColor=&VC_DynamicScaleColor(
     Color to Scale:LinearColor=$Particles.Initial.Color
     RGB Scale:Vector3f=<0.5 0.5 0.5>
     Alpha Scale:NiagaraFloat=&FloatFromCurve(FloatCurve=DataInterface() {propertyValues {Curve {keys [<at 0 0>,<at 1 3>]}}})
    )
    Particles.Tentacle5_Position:NiagaraPosition=&AddVectorToPosition(Position=$Particles.Position, Vector=&VC_ConvertVectorToWorld(Vector:Vector3f=<-25 5 5>))
    Particles.Tentacle5_Width:NiagaraFloat=&Multiply_Float(
     A=12
     B=&FloatFromCurve(FloatCurve=DataInterface() {propertyValues {Curve {keys [<at 0 1>,<at 0.4 0>]}}})
    )
   ) {isSetParameters=true}
  ]
  rendererClasses [&NiagaraRibbonRendererProperties]
  renderers [
   {rendererClass=&NiagaraRibbonRendererProperties}
   {rendererClass=&NiagaraRibbonRendererProperties, rendererIndex=1}
   {rendererClass=&NiagaraRibbonRendererProperties, rendererIndex=2}
   {rendererClass=&NiagaraRibbonRendererProperties, rendererIndex=3}
   {rendererClass=&NiagaraRibbonRendererProperties, rendererIndex=4}
  ]
 }
}
summary {
 systemName=NS_Boost_Jellyfish
 emitters [
  {
   emitterName=Jellyfish_Entry
   rendererClasses [&NiagaraSpriteRendererProperties, &NiagaraMeshRendererProperties]
  }
  {emitterName=TentaclesLarge, rendererClasses [&NiagaraRibbonRendererProperties]}
  {emitterName=NE_VC_SingleBoost_Sprite_Parent, rendererClasses [&NiagaraSpriteRendererProperties]}
  {emitterName=BackRibbon, rendererClasses [&NiagaraRibbonRendererProperties]}
 ]
 userVariables [
  User.bIsFrontendPreview {type=NiagaraBool}
  User.bIsPainted {type=NiagaraBool}
  User.ColorVariantID {type=NiagaraInt32}
  User.CustomColor {type=LinearColor, description="", defaultValue {struct=LinearColor, value=<rgba 1 1 1 1>}}
  User.DTCarOffset {type=Vector3f, description="", defaultValue {struct=Vector3f, value=<0 0 0>}}
  User.GameModeID {type=NiagaraInt32}
  User.PaintedColor {type=LinearColor, description="", defaultValue {struct=LinearColor, value=<rgba 1 1 1 1>}}
  User.PaintedScalar {type=NiagaraFloat, defaultValue=1, description=""}
  User.PortCount {type=NiagaraInt32, defaultValue=1}
  User.SingleBoostPos {type=NiagaraPosition, description="", defaultValue {}}
 ]
}
