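# Library material function for VFX that builds a per-axis scale offset vector around an object's local pivot.
# It converts world position minus the mesh particle pivot location into local space, and also dots that with the instance-space object orientation.
# Named reroutes carry these pivot values and the Transform (scale direction) input into per-component multiplies, recombined as the float3 output.
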MaterialFunction /DealershipSharedLibrary/FX/Materials/MaterialFunctions/MF_VCVFX_ObjectScaleLocalPivot_
graph {
 properties {bExposeToLibrary=true}
 nodes [
  n1 = FunctionInput() {bUsePreviewValueAsDefault=true, description=Scale Direction, inputName=Transform}
  n2 = MaterialFunctionCall(X=ComponentMask(None=NamedRerouteDeclaration(None=@n1) {name=Transform In, nodeColor=<rgba 0.6172 0 1 1>}) {r=true}) {materialFunction=&MakeFloat3}
  n3 = ComponentMask(None=NamedRerouteUsage()) {r=true}
  n4 = MaterialFunctionCall(Y=ComponentMask(None=NamedRerouteUsage()) {g=true}) {materialFunction=&MakeFloat3}
  n5 = ComponentMask(None=NamedRerouteUsage()) {g=true}
  n6 = MaterialFunctionCall(Z=ComponentMask(None=NamedRerouteUsage()) {b=true}) {materialFunction=&MakeFloat3}
  n7 = NamedRerouteUsage()
  FunctionOutput(None=MaterialFunctionCall(X=Multiply(A=@n2.Result, B=@n3), Y=Multiply(A=@n4.Result, B=@n5), Z=Multiply(A=@n6.Result, B=@n7)) {materialFunction=&MakeFloat3}.Result)
  n8 = Transform(None=Subtract(
    A=WorldPosition().XYZ
    B=MaterialFunctionCall() {materialFunction=&ObjectPivotPoint}.Mesh Particle Pivot Location
   )) {transformSourceType=TRANSFORMSOURCE_World, transformType=TRANSFORM_Local}
  NamedRerouteDeclaration(None=@n8) {name=Pivot, nodeColor=<rgba 1 0.9375 0 1>}
  n9 = ObjectOrientation()
  NamedRerouteDeclaration(None=DotProduct(A=@n8, B=Transform(None=@n9) {transformSourceType=TRANSFORMSOURCE_World, transformType=TRANSFORM_Instance})) {name=Dot Pivot, nodeColor=<rgba 0 0.3437 1 1>}
 ]
}
