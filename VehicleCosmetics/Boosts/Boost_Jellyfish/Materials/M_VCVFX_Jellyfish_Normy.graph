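# Masked unlit Niagara mesh-particle material for the Jellyfish vehicle boost, giving a glossy, paintable look with tentacle-texture glow.
# Emissive layers fake specular, edge and core fresnel, and tinted cubemap reflections with hue and lightness shifts.
# The Painted dynamic parameter blends particle color toward custom colors.
# A time-driven sine along instance-space X builds a pivot-relative ripple offset masked by vertex red, and Perlin noise builds a dissolve mask.
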Material /VehicleCosmetics/Boosts/Boost_Jellyfish/Materials/M_VCVFX_Jellyfish_Normy
graph {
 properties {
  bAutomaticallySetUsageInEditor=false
  bUsedWithNiagaraMeshParticles=true
  blendMode=BLEND_Masked
  opacityMaskClipValue=0.5
  shadingModel=MSM_Unlit
  usedShadingModels=Unlit
 }
 nodes [
  MaterialExpressionMeshEmitterVertexColor_3 = ParticleColor()
  MaterialExpressionMeshEmitterDynamicParameter_1 = DynamicParameter() {paramNames [Painted, Glow, Param3, CustomWhiteRange]}
  n1 = LinearInterpolate(
   A=@MaterialExpressionMeshEmitterVertexColor_3.RGB
   B=VectorParameter() {parameterName=FadeColor, defaultValue=<rgba 0.4917 0.4917 0.4917 1>}.RGB
   Alpha=@MaterialExpressionMeshEmitterDynamicParameter_1.Painted
  )
  n2 = TextureSample() {texture=/DealershipSharedLibrary/FX/Textures/TexturePackage/T_VCVFX_JellyfishTentacle_RGB}
  MaterialExpressionConstantClamp_16 = Saturate(None=Multiply(A=ScalarParameter() {defaultValue=1, group=reflection, parameterName=ReflectionBrightness}, B=NamedRerouteUsage()))
  n3 = DotProduct(
   A=Transform(None=CameraVectorWS()) {transformSourceType=TRANSFORMSOURCE_World, transformType=TRANSFORM_Tangent}
   B=Constant3Vector() {constant=<rgba 0 0 1 0>}
  )
  MaterialExpressionConstantClamp_29 = Saturate(None=ConstantBiasScale(None=@n3) {bias=-0.4, scale=2})
  n4 = Add(A=@MaterialExpressionConstantClamp_29, B=Constant() {r=0.15})
  n5 = ScalarParameter() {defaultValue=0.8, group=Specular, parameterName=SpecularAmount}
  n6 = Multiply(
   A=Multiply(A=ScalarParameter() {defaultValue=0.65, group=Specular, parameterName=SpecBrightness}, B=Power(Base=@n4, Exp=@n5))
   B=NamedRerouteUsage()
  )
  n7 = Multiply(A=ScalarParameter() {defaultValue=1, group=reflection, parameterName=EdgeReflection}, B=NamedRerouteUsage())
  MaterialExpressionConstantClamp_25 = Saturate(None=ConstantBiasScale(None=OneMinus(None=@n3)) {bias=-0.3, scale=3})
  n8 = Add(A=@MaterialExpressionConstantClamp_25, B=Constant() {r=0.05})
  n9 = ScalarParameter() {defaultValue=0.5, parameterName=ShiftIntensity}
  n10 = Multiply(
   A=MaterialFunctionCall!(Shift=Multiply(A=@n8, B=@n9), Color=NamedRerouteUsage()) {materialFunction=/DealershipSharedLibrary/FX/Materials/MaterialFunctions/MF_VCVFX_LightnessShift}.Result
   B=Multiply(A=@MaterialExpressionConstantClamp_25, B=ScalarParameter() {defaultValue=10, group=Fresnel, parameterName=EdgeFresnelBrightness})
  )
  n11 = Multiply(A=ScalarParameter() {defaultValue=1, group=reflection, parameterName=CoreReflection}, B=NamedRerouteUsage())
  MaterialExpressionConstantClamp_28 = Saturate(None=ConstantBiasScale(None=@n3) {bias=-0.6, scale=1})
  n12 = Multiply(
   A=Multiply(A=@MaterialExpressionConstantClamp_28, B=ScalarParameter() {defaultValue=2, group=Fresnel, parameterName=CoreFresnelBrightness})
   B=@n6
  )
  n13 = NamedRerouteUsage()
  n14 = Add(A=Add(A=Multiply(A=@n7, B=@n10), B=Multiply(A=@n11, B=@n12)), B=Multiply(A=Desaturation(None=@n13), B=NamedRerouteUsage()))
  n15 = Add(A=@n6, B=@n14)
  n16 = Add(A=@MaterialExpressionConstantClamp_25, B=@MaterialExpressionConstantClamp_28)
  n17 = Multiply(A=@n16, B=@n16)
  n18 = Constant() {r=0.2}
  n19 = LinearInterpolate(
   A=Multiply(A=NamedRerouteUsage(), B=Add(A=@n17, B=@n18))
   B=@n14
   Alpha=ScalarParameter() {defaultValue=2.5, group=reflection, parameterName=ReflectionTint}
  )
  n20 = Add(A=Multiply(A=@MaterialExpressionConstantClamp_16, B=@n15), B=@n19)
  n21 = ScalarParameter() {defaultValue=1, parameterName=Glow}
  MaterialFunctionCall(Emissive=Add(A=Multiply(A=@n1, B=@n2.B), B=Multiply(A=@n20, B=@n21))) {materialFunction=&MF_FX_EmissiveAdjustments}
  MaterialExpressionObjectWorldPosition_1 = ObjectPositionWS()
  n22 = WorldPosition()
  n23 = ComponentMask(None=Transform(None=Subtract(A=@MaterialExpressionObjectWorldPosition_1, B=@n22.XYZ)) {transformSourceType=TRANSFORMSOURCE_World, transformType=TRANSFORM_Instance}) {r=true}
  n24 = Add(
   A=Multiply(A=OneMinus(None=@n23), B=ScalarParameter() {defaultValue=0.07, group=head, parameterName=RippleIntensity})
   B=MaterialFunctionCall(Speed=ScalarParameter() {defaultValue=0.3, group=head, parameterName=RippleSpeed}) {materialFunction=&TimeWithSpeedVariable}.Time
  )
  n25 = NamedRerouteDeclaration(None=Sine(None=NamedRerouteDeclaration(None=@n24) {name=WPO Add, nodeColor=<rgba 0 1 0.7422 1>}) {period=0.75}) {name=WPO SIne, nodeColor=<rgba 1 0.4219 0 1>}
  n26 = Multiply(
   A=MaterialFunctionCall(X=@n25, Y=@n25, Z=@n25) {materialFunction=&MakeFloat3}.Result
   B=VectorParameter() {group=head, parameterName=HeadWPO, defaultValue=<rgba 0.05 0.17 0.17 1>}.RGB
  )
  n27 = Subtract(
   A=WorldPosition().XYZ
   B=MaterialFunctionCall() {materialFunction=&ObjectPivotPoint}.Mesh Particle Pivot Location
  )
  Multiply(
   A=VertexColor().R
   B=Add(
    A=MaterialFunctionCall!(Transform=@n26) {materialFunction=/DealershipSharedLibrary/FX/Materials/MaterialFunctions/MF_VCVFX_ObjectScaleLocalPivot_}.Result
    B=Multiply(A=@n27, B=@n26)
   )
  )
  n28 = ParticleColor()
  NamedRerouteDeclaration(None=LinearInterpolate(
    A=@n28.RGB
    B=VectorParameter() {group=Colors, parameterName=CustomColor, defaultValue=<rgba 0.0385 0.231 1 1>}.RGB
    Alpha=DynamicParameter() {paramNames [Painted, Glow, Param3, CustomWhiteRange]}.Painted
   )) {name=LerpCustomColor1, nodeColor=<rgba 0 1 0.0625 1>}
  NamedRerouteDeclaration(None=Add(A=@n19, B=@n15)) {name=Add1, nodeColor=<rgba 0 0.2969 1 1>}
  n29 = ReflectionVectorWS()
  MaterialExpressionConstantClamp_26 = Saturate(None=Transform(None=@n29) {transformSourceType=TRANSFORMSOURCE_World, transformType=TRANSFORM_Tangent})
  MaterialExpressionConstantClamp_27 = Saturate(None=Transform(None=CameraVectorWS()) {transformSourceType=TRANSFORMSOURCE_World, transformType=TRANSFORM_Tangent})
  n30 = Add(
   A=ComponentMask(None=@MaterialExpressionConstantClamp_26) {g=true, r=true}
   B=ComponentMask(None=@MaterialExpressionConstantClamp_27) {g=true, r=true}
  )
  n31 = ComponentMask(None=@n30) {r=true}
  n32 = NamedRerouteUsage()
  n33 = LinearInterpolate(
   A=@n28.RGB
   B=VectorParameter() {group=Colors, parameterName=CoreColor, defaultValue=<rgba 1 0 0.5 1>}.RGB
   Alpha=DynamicParameter() {paramNames [Painted, Glow, Param3, CustomWhiteRange]}.Painted
  )
  n34 = ComponentMask(None=@n30) {g=true}
  NamedRerouteDeclaration(None=Add(A=Multiply(A=@n31, B=@n32), B=Multiply(A=@n33, B=@n34))) {name=Ref Vector Add, nodeColor=<rgba 1 0.2578 0 1>}
  n35 = Transform(None=ReflectionVectorWS()) {transformSourceType=TRANSFORMSOURCE_Local}
  n36 = TextureSampleParameterCube!(UVs=@n35, Apply View MipBias) {group=reflection, parameterName=CubeMap_01, texture=/DealershipSharedLibrary/FX/Textures/Cubemaps/Cubemap_VCVFX_City_CustomTeam02_}
  MaterialExpressionConstantClamp_18 = Saturate(None=ConstantBiasScale(None=@n36.RGB) {bias=0.01, scale=2})
  MaterialExpressionConstantClamp_19 = Saturate(None=ConstantBiasScale(None=TextureSampleParameterCube(UVs=@n35) {group=reflection, parameterName=CubeMap_02, texture=/DealershipSharedLibrary/FX/Textures/Cubemaps/Cubemap_VCVFX_Rainbow_}.RGB) {bias=0.01, scale=2})
  n37 = Add(A=@MaterialExpressionConstantClamp_18, B=@MaterialExpressionConstantClamp_19)
  n38 = Sine(None=NamedRerouteUsage())
  MaterialExpressionConstantClamp_21 = Saturate(None=Multiply(A=@n38, B=@n38))
  MaterialExpressionConstantClamp_20 = Saturate(None=Add(A=Multiply(B=@MaterialExpressionConstantClamp_21) {constA=0.25}) {constB=0.3})
  n39 = DotProduct(
   A=Transform(None=CameraVectorWS()) {transformSourceType=TRANSFORMSOURCE_World, transformType=TRANSFORM_Tangent}
   B=Constant3Vector() {constant=<rgba 0 0 1 0>}
  )
  MaterialExpressionConstantClamp_23 = Saturate(None=ConstantBiasScale(None=OneMinus(None=@n39)) {bias=-0.3, scale=3})
  MaterialExpressionConstantClamp_22 = Saturate(None=Add(A=Constant() {r=0.5}, B=@MaterialExpressionConstantClamp_23))
  MaterialExpressionConstantClamp_24 = Saturate(None=NamedRerouteUsage())
  n40 = Multiply(A=Constant() {r=0.1}, B=@MaterialExpressionConstantClamp_24)
  n41 = StaticSwitchParameter(
   True=MaterialFunctionCall(Shift=@MaterialExpressionConstantClamp_20, Color=NamedRerouteUsage()) {materialFunction=&MF_VCVFX_LightnessShift}.Result
   False=MaterialFunctionCall(
    Shift=@MaterialExpressionConstantClamp_22
    Color=MaterialFunctionCall(Hue Shift Percentage=@n40, Texture=@n32) {materialFunction=&HueShift}.Result
   ) {materialFunction=&MF_VCVFX_LightnessShift}.Result
  ) {parameterName=Painted}
  n42 = MaterialFunctionCall!(
   Desaturate=ScalarParameter() {defaultValue=0.1, parameterName=WorldEnvSaturation}
   "Tint (Curve)"=@n41
   Tint Amount=Constant() {r=0.25}
   Cubemap=TextureObject() {texture=/DealershipSharedLibrary/FX/Textures/Cubemaps/Cubemap_VCVFX_Green_}
  ) {materialFunction=/DealershipSharedLibrary/FX/Materials/MaterialFunctions/VC_MF_WorldEnv}
  n43 = LinearInterpolate(Alpha=DynamicParameter() {paramNames [Painted, Glow, Param3, CustomWhiteRange]}.CustomWhiteRange) {constA=0.5}
  n44 = Desaturation(None=Add(A=Multiply(A=@n37, B=@n42.LocalSpace), B=@n42.LocalSpace), Fraction=@n43)
  n45 = ScalarParameter() {defaultValue=1.25, group=reflection, parameterName=ReflectionPower}
  MaterialExpressionConstantClamp_17 = Saturate(None=Multiply(A=Power(Base=@n44, Exp=@n45), B=ScalarParameter() {defaultValue=2, group=reflection, parameterName=ReflectionAmount}))
  NamedRerouteDeclaration(None=LinearInterpolate(A=@MaterialExpressionConstantClamp_17, B=Multiply(A=@MaterialExpressionConstantClamp_17, B=@n41), Alpha=@n43)) {name=Lerp Cubemaps, nodeColor=<rgba 0 0.0625 1 1>}
  n46 = TextureCoordinate() {uTiling=0.8, vTiling=1.3}
  MaterialExpressionConstantClamp_15 = Saturate(None=Subtract(
    A=ScalarParameter() {defaultValue=1.5, parameterName=Dissolve}
    B=TextureSample(UVs=@n46) {texture=/DealershipSharedLibrary/FX/Textures/Noise/T_VCVFX_Boost_S16_PerlinBlur}.RGB
   ))
  Reroute(None=@MaterialExpressionConstantClamp_15)
  ScalarParameter() {group=reflection, parameterName=ReflectionSaturation}
  Transform(None=@n29)
 ]
}
