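# Unlit translucent Niagara ribbon material for the tentacles of the Jellyfish vehicle boost effect.
# A timed sine/cosine wave along UV.x, weighted by a power of that coordinate, builds a world-position-offset wiggle.
# A channel-selectable mask, scaled by dynamic-parameter opacity and depth-faded, forms opacity.
# Emissive blends gradient-driven particle and custom color layers, with named reroutes carrying the dynamic Light, GradientCurve and Painted values.
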Material /VehicleCosmetics/Boosts/Boost_Jellyfish/Materials/M_VCVFX_Jellyfish_Tentacle
graph {
 properties {
  bAutomaticallySetUsageInEditor=false
  bUsedWithNiagaraRibbons=true
  blendMode=BLEND_Translucent
  shadingModel=MSM_Unlit
  usedShadingModels=Unlit
 }
 nodes [
  Add(A=TextureCoordinate(), B=Constant2Vector() {r=-0.1})
  Constant()
  n1 = TextureSample() {texture=/DealershipSharedLibrary/FX/Textures/TexturePackage/T_VCVFX_JellyfishTentacle_RGB}
  LinearInterpolate(A=@n1.R) {constA=1, constAlpha=1}
  MaterialFunctionCall(Emissive=NamedRerouteUsage()) {materialFunction=&MF_FX_EmissiveAdjustments}
  n2 = TextureCoordinate()
  n3 = VectorParameter() {group=WPO, parameterName=WPO Parameters, channelNames {a=WPO Strength 2, b=WPO Strength, g=Speed, r=Tile}}
  n4 = Add(
   A=Divide(A=ComponentMask(None=@n2) {r=true}, B=@n3.Tile)
   B=MaterialFunctionCall(Speed=@n3.Speed) {materialFunction=&TimeWithSpeedVariable}.Time
  )
  n5 = Sine(None=@n4)
  n6 = MaterialFunctionCall(X=Cosine(None=@n4), Y=@n5, Z=@n5) {materialFunction=&MakeFloat3}
  n7 = Multiply(A=Multiply(A=Transform(None=@n6.Result) {transformSourceType=TRANSFORMSOURCE_Local}, B=@n3.WPO Strength), B=@n3.WPO Strength 2)
  n8 = TextureCoordinate()
  n9 = Multiply(
   A=ConstantBiasScale(None=Transform(None=@n7) {transformSourceType=TRANSFORMSOURCE_World, transformType=TRANSFORM_Local})
   B=Power(Base=ComponentMask(None=@n8) {r=true}, Exp=ScalarParameter() {group=WPO, parameterName=WPOPower})
  )
  NamedRerouteDeclaration(None=@n9) {name=WPO, nodeColor=<rgba 1 0 0.8437 1>}
  n10 = DynamicParameter() {paramNames [Opacity, Light, GradientCurve, Painted]}
  NamedRerouteDeclaration(None=@n10.Light) {name=DYN LIGHT, nodeColor=<rgba 1 0 0.2578 1>}
  NamedRerouteDeclaration(None=@n10.GradientCurve) {name=DYN GradientCurve, nodeColor=<rgba 0.7344 1 0 1>}
  NamedRerouteDeclaration(None=@n10.Painted) {name=DYN ISPAINTED?, nodeColor=<rgba 0.9453 0 1 1>}
  NamedRerouteDeclaration(None=Multiply(A=NamedRerouteUsage(), B=@n1.R)) {name=Texture Mask, nodeColor=<rgba 1 0 0.6562 1>}
  n11 = VectorParameter() {parameterName=CustomColor02, defaultValue=<rgba 0.3084 0.5153 1 4>}
  NamedRerouteDeclaration(None=Multiply(A=@n11.RGB, B=@n11.RGB)) {name=Color2, nodeColor=<rgba 0 0.5781 1 1>}
  n12 = VectorParameter() {parameterName=CustomColor, defaultValue=<rgba 0.5439 0.1547 1 4>}
  NamedRerouteDeclaration(None=Multiply(A=@n12.RGB, B=@n12.RGB)) {name=Color1, nodeColor=<rgba 0.8984 1 0 1>}
  n13 = TextureSampleParameter2D() {
   group=Opacity
   parameterName=MaskTexture
   sortPriority=1
   texture=/DealershipSharedLibrary/FX/Textures/TexturePackage/T_VCVFX_Jellyfish_RGB01
  }
  n14 = DotProduct(A=@n13.RGB, B=VectorParameter() {group=Opacity, parameterName=RGB Channel Selector, sortPriority=2, defaultValue=<rgba 1 0 0 0>}.RGB)
  n15 = Constant() {r=1}
  MaterialExpressionConstantClamp_0 = Saturate(None=Multiply(A=Desaturation(None=@n14, Fraction=@n15), B=@n10.Opacity))
  NamedRerouteDeclaration(None=MaterialFunctionCall(
    Opacity In=@MaterialExpressionConstantClamp_0
    DepthFade Distance=ScalarParameter() {defaultValue=35, group=Opacity, parameterName=DepthFadeDistance}
   ) {materialFunction=&MF_VCVFX_DepthFade}.Result) {name=Opacity, nodeColor=<rgba 0.4522 0.4522 0.4522 1>}
  n16 = Desaturation(None=TextureSample() {samplerSource=SSM_Wrap_WorldGroupSettings, texture=/DealershipSharedLibrary/FX/Textures/TexturePackage/T_VCVFX_S8RBSG_Gradient_RGB}.RGB)
  n17 = NamedRerouteUsage()
  MaterialExpressionConstantClamp_1 = Saturate(None=ConstantBiasScale(None=Multiply(A=@n16, B=@n17)) {bias=-0.001, scale=3})
  n18 = LinearInterpolate(A=NamedRerouteUsage(), B=NamedRerouteUsage(), Alpha=@MaterialExpressionConstantClamp_1)
  n19 = NamedRerouteUsage()
  n20 = Constant() {r=2}
  MaterialExpressionVertexColor_0 = ParticleColor()
  n21 = NamedRerouteUsage()
  n22 = LinearInterpolate(
   A=Multiply(A=Multiply(A=@n20, B=@MaterialExpressionVertexColor_0.RGB), B=@n21)
   B=Multiply(A=@n21, B=@MaterialExpressionVertexColor_0.RGB) {constA=2}
   Alpha=@MaterialExpressionConstantClamp_1
  )
  NamedRerouteDeclaration(None=Multiply(A=NamedRerouteUsage(), B=LinearInterpolate(A=@n22, B=@n18, Alpha=@n19))) {name=Emissive, nodeColor=<rgba 0 1 0.4609 1>}
  NamedRerouteUsage()
  NamedRerouteUsage()
  Transform(None=@n9) {transformSourceType=TRANSFORMSOURCE_World, transformType=TRANSFORM_Local}
 ]
}
